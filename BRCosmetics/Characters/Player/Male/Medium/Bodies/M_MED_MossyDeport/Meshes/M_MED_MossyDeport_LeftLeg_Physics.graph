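# Physics asset for the left-leg mesh of the male medium MossyDeport body, setting up physics for its dangling front and back tassels.
# Each tassel is a two-bone capsule chain hung from a small master body whose angular motion is locked to the pelvis.
# The chain joints carry small swing limit angles and zero twist angle, with wider swing values on the back tassel.
# Unconstrained thigh-twist and pelvis capsules provide body volumes around the tassels.
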PhysicsAsset /BRCosmetics/Characters/Player/Male/Medium/Bodies/M_MED_MossyDeport/Meshes/M_MED_MossyDeport_LeftLeg_Physics
bodies {
 dyn_backTassel_l_01 [{shapeType=Capsule, length=2.5, radius=1.5, shapeName=dyn_backTassel_l_01_capsule, center=<2.4614 0 0>, rotation=<rot 0 -90 -90>}]
 dyn_backTassel_l_02 [{shapeType=Capsule, length=2, radius=2, shapeName=dyn_backTassel_l_02_capsule, center=<2.1706 0 0>, rotation=<rot 0 90 -90>}]
 dyn_backTassel_l_MSTR [{shapeType=Capsule, length=0.505, radius=0.505, shapeName=dyn_backTassel_l_MSTR_capsule, rotation=<rot 0 -90 0>}]
 dyn_frontTassel_l_01 [{shapeType=Capsule, length=2.5, radius=1.5, shapeName=dyn_frontTassel_l_01_capsule, center=<2.4614 0 0>, rotation=<rot 0 90 -90>}]
 dyn_frontTassel_l_02 [{shapeType=Capsule, length=2, radius=2, shapeName=dyn_frontTassel_l_02_capsule, center=<2.1706 0 0>, rotation=<rot 0 90 -90>}]
 dyn_frontTassel_l_MSTR [{shapeType=Capsule, length=0.5, radius=0.5, shapeName=dyn_frontTassel_l_MSTR_capsule, rotation=<rot 0 -90 0>}]
 pelvis [
  {shapeType=Capsule, length=14.5, radius=13, center=<-10.058 -0.1096 0>, rotation=<rot 0 -90 0>}
  {shapeType=Capsule, length=9, radius=10.5, shapeName=pelvis_capsule, center=<-11.2405 -3.2539 0.0557>, rotation=<rot 0 97.3362 90>}
 ]
 thigh_twist_01_l [{shapeType=Capsule, length=29.0745, radius=13.3435, center=<-8.8355 -0.721 -2.6115>, rotation=<rot 2.0155 -84.3305 -87.8329>}]
}
constraints [
 {
  bone1Name=dyn_backTassel_l_MSTR
  bone2Name=pelvis
  swing1LimitDegrees=45
  swing1Motion=Locked
  swing2LimitDegrees=45
  swing2Motion=Locked
  twistLimitDegrees=45
  twistMotion=Locked
 }
 {bone1Name=dyn_backTassel_l_01, bone2Name=dyn_backTassel_l_MSTR, swing1LimitDegrees=10, swing2LimitDegrees=20, twistLimitDegrees=0}
 {bone1Name=dyn_backTassel_l_02, bone2Name=dyn_backTassel_l_01, swing1LimitDegrees=20, swing2LimitDegrees=30, twistLimitDegrees=0}
 {
  bone1Name=dyn_frontTassel_l_MSTR
  bone2Name=pelvis
  swing1LimitDegrees=45
  swing1Motion=Locked
  swing2LimitDegrees=45
  swing2Motion=Locked
  twistLimitDegrees=45
  twistMotion=Locked
 }
 {bone1Name=dyn_frontTassel_l_01, bone2Name=dyn_frontTassel_l_MSTR, swing1LimitDegrees=10, swing2LimitDegrees=10, twistLimitDegrees=0}
 {bone1Name=dyn_frontTassel_l_02, bone2Name=dyn_frontTassel_l_01, swing1LimitDegrees=10, swing2LimitDegrees=10, twistLimitDegrees=0}
]
