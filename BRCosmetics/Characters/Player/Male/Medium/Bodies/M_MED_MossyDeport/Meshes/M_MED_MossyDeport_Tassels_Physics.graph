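# Physics asset for the back tassels on the M_MED_MossyDeport male medium player cosmetic body.
# Left and right tassels are each four-capsule chains with small swing limits, zero-degree twist limits and radii that widen toward the tip.
# Each chain hangs from its own master body, which is angularly locked to a central master that is in turn locked to the pelvis.
# Two large capsules form the pelvis body.
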PhysicsAsset /BRCosmetics/Characters/Player/Male/Medium/Bodies/M_MED_MossyDeport/Meshes/M_MED_MossyDeport_Tassels_Physics
bodies {
 dyn_backTassel_c_01_l [{shapeType=Capsule, length=3.5, radius=0.8, shapeName=dyn_backTassel_c_01_l_capsule, center=<3 0.1261 0>, rotation=<rot 0 90 -90>}]
 dyn_backTassel_c_01_r [{shapeType=Capsule, length=3.5, radius=0.8, shapeName=dyn_backTassel_c_01_r_capsule, center=<3 0.1261 0>, rotation=<rot 0 90 -90>}]
 dyn_backTassel_c_02_l [{shapeType=Capsule, length=3.5, radius=1, shapeName=dyn_backTassel_c_02_l_capsule, center=<2.6859 0 0.1945>, rotation=<rot 0 90 -90>}]
 dyn_backTassel_c_02_r [{shapeType=Capsule, length=3.5, radius=1, shapeName=dyn_backTassel_c_02_r_capsule, center=<2.6859 0 0.1945>, rotation=<rot 0 -90 -90>}]
 dyn_backTassel_c_03_l [{shapeType=Capsule, length=2.5, radius=1.25, shapeName=dyn_backTassel_c_03_l_capsule, center=<1.8141 0.4286 0>, rotation=<rot 0 -90 -90>}]
 dyn_backTassel_c_03_r [{shapeType=Capsule, length=2.5, radius=1.25, shapeName=dyn_backTassel_c_03_r_capsule, center=<1.8141 -0.4286 0>, rotation=<rot 0 -90 -90>}]
 dyn_backTassel_c_04_l [{shapeType=Capsule, length=1, radius=2, shapeName=dyn_backTassel_c_04_l_capsule, center=<2.4 0 0>, rotation=<rot 0 90 -90>}]
 dyn_backTassel_c_04_r [{shapeType=Capsule, length=0.5, radius=2.5, shapeName=dyn_backTassel_c_04_r_capsule, center=<2.4393 0 0>, rotation=<rot 0 90 -90>}]
 dyn_backTassel_c_MSTR [{shapeType=Capsule, length=0.5, radius=0.5, shapeName=dyn_backTassel_c_MSTR_capsule, rotation=<rot 0 -90 0>}]
 dyn_backTassel_c_l_MSTR [{shapeType=Capsule, length=0.5, radius=0.5, shapeName=dyn_backTassel_c_l_MSTR_capsule, rotation=<rot 0 -90 0>}]
 dyn_backTassel_c_r_MSTR [{shapeType=Capsule, length=0.505, radius=0.505, shapeName=dyn_backTassel_c_r_MSTR_capsule, rotation=<rot 0 -90 0>}]
 pelvis [
  {shapeType=Capsule, length=14.5, radius=13, center=<-10.058 -0.1096 0>, rotation=<rot 0 -90 0>}
  {shapeType=Capsule, length=9, radius=10.5, shapeName=pelvis_capsule, center=<-11.2405 -3.2539 0.0557>, rotation=<rot 0 97.3362 90>}
 ]
}
constraints [
 {bone1Name=dyn_backTassel_c_01_l, bone2Name=dyn_backTassel_c_l_MSTR, swing1LimitDegrees=15, swing2LimitDegrees=10, twistLimitDegrees=0}
 {bone1Name=dyn_backTassel_c_02_l, bone2Name=dyn_backTassel_c_01_l, swing1LimitDegrees=20, swing2LimitDegrees=10, twistLimitDegrees=0}
 {bone1Name=dyn_backTassel_c_03_l, bone2Name=dyn_backTassel_c_02_l, swing1LimitDegrees=20, swing2LimitDegrees=20, twistLimitDegrees=0}
 {bone1Name=dyn_backTassel_c_04_l, bone2Name=dyn_backTassel_c_03_l, swing1LimitDegrees=20, swing2LimitDegrees=20, twistLimitDegrees=0}
 {bone1Name=dyn_backTassel_c_01_r, bone2Name=dyn_backTassel_c_r_MSTR, swing1LimitDegrees=15, swing2LimitDegrees=10, twistLimitDegrees=0}
 {bone1Name=dyn_backTassel_c_02_r, bone2Name=dyn_backTassel_c_01_r, swing1LimitDegrees=15, swing2LimitDegrees=20, twistLimitDegrees=0}
 {bone1Name=dyn_backTassel_c_03_r, bone2Name=dyn_backTassel_c_02_r, swing1LimitDegrees=20, swing2LimitDegrees=20, twistLimitDegrees=0}
 {bone1Name=dyn_backTassel_c_04_r, bone2Name=dyn_backTassel_c_03_r, swing1LimitDegrees=20, swing2LimitDegrees=20, twistLimitDegrees=0}
 {
  bone1Name=dyn_backTassel_c_MSTR
  bone2Name=pelvis
  swing1LimitDegrees=45
  swing1Motion=Locked
  swing2LimitDegrees=45
  swing2Motion=Locked
  twistLimitDegrees=45
  twistMotion=Locked
 }
 {
  bone1Name=dyn_backTassel_c_r_MSTR
  bone2Name=dyn_backTassel_c_MSTR
  swing1LimitDegrees=45
  swing1Motion=Locked
  swing2LimitDegrees=45
  swing2Motion=Locked
  twistLimitDegrees=45
  twistMotion=Locked
 }
 {
  bone1Name=dyn_backTassel_c_l_MSTR
  bone2Name=dyn_backTassel_c_MSTR
  swing1LimitDegrees=10
  swing1Motion=Locked
  swing2LimitDegrees=10
  swing2Motion=Locked
  twistLimitDegrees=0
  twistMotion=Locked
 }
]
