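# Post-parent AnimBP for the MossyDeport male body cosmetic, adding secondary motion to vest, pants, anklets and tassels.
# Copied parent pose runs through a Control Rig, AnimDynamics for anklets and pants, then RigidBody sims for legs and tassels, into a cache.
# A bool blend picks between vest AnimDynamics chains with or without back bones.
# EventGraph updates gravity override and front-end alpha, and sets backbling pack/cape flags on customization load.
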AnimBlueprint /BRCosmetics/Characters/Player/Male/Medium/Bodies/M_MED_MossyDeport/Meshes/M_MED_MossyDeport_AnimBP
target_skeleton=&/M_MED_MossyDeport_Skeleton
graphs [
 AnimGraph {nodes [
   n1 = Anim Dynamics - Bone: dyn_vest_front_l_01!(ComponentPose=Local To Component(LocalPose=Use cached pose 'startCache'()), Alpha=1, GravityOverride:Vector="0,0,-980") -> (Pose) {
    angularDampingOverride=1
    angularSpringConstant=20000
    bAngularSpring=true
    bOverrideAngularDamping=true
    bUseGravityOverride=true
    gravityScale=0.35
    simulationSpace=RootRelative
    boundBone {boneName=dyn_vest_front_l_01}
    componentAppliedLinearAccClamp=<4000 4000 4000>
    componentLinearAccScale=<0.7 0.8 0.7>
    componentLinearVelScale=<0.7 0.8 0.7>
    gravityOverride {z=-980}
    physicsBodyDefinitions [{
      boundBone {boneName=dyn_vest_front_l_01}
      boxExtents=<20 30 30>
      constraintSetup {angularTargetAxis=AxisY, angularLimitsMax=<0 0 10>, angularLimitsMin=<0 -90 0>, angularTarget=<1 0 0>}
      localJointOffset=<4 0 0>
     }]
    relativeSpaceBone {boneName=spine_04}
    simSpaceSettings {maxAngularAcceleration=100, maxAngularVelocity=10, simSpaceAngularAlpha=1}
   }
   n2 = Anim Dynamics - Bone: dyn_vest_side_l_01!(ComponentPose=@n1, Alpha=1, GravityOverride:Vector="0,0,-980") -> (Pose) {
    angularDampingOverride=1
    angularSpringConstant=20000
    bAngularSpring=true
    bOverrideAngularDamping=true
    bUseGravityOverride=true
    gravityScale=0.35
    simulationSpace=RootRelative
    boundBone {boneName=dyn_vest_side_l_01}
    componentAppliedLinearAccClamp=<4000 4000 4000>
    componentLinearAccScale=<0.7 0.8 0.7>
    componentLinearVelScale=<0.7 0.8 0.7>
    gravityOverride {z=-980}
    physicsBodyDefinitions [{
      boundBone {boneName=dyn_vest_side_l_01}
      boxExtents=<20 30 30>
      constraintSetup {angularTargetAxis=AxisY, angularLimitsMax=<0 5 10>, angularLimitsMin=<0 -90 -10>, angularTarget=<1 0 0>}
      localJointOffset=<4 0 0>
     }]
    relativeSpaceBone {boneName=spine_04}
    simSpaceSettings {maxAngularAcceleration=100, maxAngularVelocity=10, simSpaceAngularAlpha=1}
   }
   n3 = Anim Dynamics - Bone: dyn_vest_back_l_01!(ComponentPose=@n2, Alpha=1, GravityOverride:Vector="0,0,-980") -> (Pose) {
    angularDampingOverride=0.999
    angularSpringConstant=12000
    bAngularSpring=true
    bOverrideAngularDamping=true
    bUseGravityOverride=true
    gravityScale=0.35
    simulationSpace=RootRelative
    boundBone {boneName=dyn_vest_back_l_01}
    componentAppliedLinearAccClamp=<4000 4000 4000>
    componentLinearAccScale=<0.7 0.8 0.7>
    componentLinearVelScale=<0.7 0.8 0.7>
    gravityOverride {z=-980}
    physicsBodyDefinitions [{
      boundBone {boneName=dyn_vest_back_l_01}
      boxExtents=<20 30 30>
      constraintSetup {angularTargetAxis=AxisY, angularLimitsMax=<0 20 5>, angularLimitsMin=<0 -120 -5>, angularTarget=<1 0 0>}
      localJointOffset=<4 0 0>
     }]
    relativeSpaceBone {boneName=spine_04}
    simSpaceSettings {maxAngularAcceleration=100, maxAngularVelocity=10, simSpaceAngularAlpha=1}
   }
   n4 = Anim Dynamics - Bone: dyn_vest_front_r_01!(ComponentPose=@n3, Alpha=1, GravityOverride:Vector="0,0,-980") -> (Pose) {
    angularDampingOverride=1
    angularSpringConstant=20000
    bAngularSpring=true
    bOverrideAngularDamping=true
    bUseGravityOverride=true
    gravityScale=0.35
    simulationSpace=RootRelative
    boundBone {boneName=dyn_vest_front_r_01}
    componentAppliedLinearAccClamp=<4000 4000 4000>
    componentLinearAccScale=<0.7 0.8 0.7>
    componentLinearVelScale=<0.7 0.8 0.7>
    gravityOverride {z=-980}
    physicsBodyDefinitions [{
      boundBone {boneName=dyn_vest_front_r_01}
      boxExtents=<20 30 30>
      constraintSetup {angularTargetAxis=AxisY, angularLimitsMax=<0 2 10>, angularLimitsMin=<0 -90 0>, angularTarget=<1 0 0>}
      localJointOffset=<4 0 0>
     }]
    relativeSpaceBone {boneName=spine_04}
    simSpaceSettings {maxAngularAcceleration=100, maxAngularVelocity=10, simSpaceAngularAlpha=1}
   }
   n5 = Anim Dynamics - Bone: dyn_vest_side_r_01!(ComponentPose=@n4, Alpha=1, GravityOverride:Vector="0,0,-980") -> (Pose) {
    angularDampingOverride=1
    angularSpringConstant=20000
    bAngularSpring=true
    bOverrideAngularDamping=true
    bUseGravityOverride=true
    gravityScale=0.35
    simulationSpace=RootRelative
    boundBone {boneName=dyn_vest_side_r_01}
    componentAppliedLinearAccClamp=<4000 4000 4000>
    componentLinearAccScale=<0.7 0.8 0.7>
    componentLinearVelScale=<0.7 0.8 0.7>
    gravityOverride {z=-980}
    physicsBodyDefinitions [{
      boundBone {boneName=dyn_vest_side_r_01}
      boxExtents=<20 30 30>
      constraintSetup {angularTargetAxis=AxisY, angularLimitsMax=<0 5 10>, angularLimitsMin=<0 -90 -10>, angularTarget=<1 0 0>}
      localJointOffset=<4 0 0>
     }]
    relativeSpaceBone {boneName=spine_04}
    simSpaceSettings {maxAngularAcceleration=100, maxAngularVelocity=10, simSpaceAngularAlpha=1}
   }
   n6 = Anim Dynamics - Bone: dyn_vest_back_r_01!(ComponentPose=@n5, Alpha=1, GravityOverride:Vector="0,0,-980") -> (Pose) {
    angularDampingOverride=0.999
    angularSpringConstant=12000
    bAngularSpring=true
    bOverrideAngularDamping=true
    bUseGravityOverride=true
    gravityScale=0.35
    simulationSpace=RootRelative
    boundBone {boneName=dyn_vest_back_r_01}
    componentAppliedLinearAccClamp=<4000 4000 4000>
    componentLinearAccScale=<0.7 0.8 0.7>
    componentLinearVelScale=<0.7 0.8 0.7>
    gravityOverride {z=-980}
    physicsBodyDefinitions [{
      boundBone {boneName=dyn_vest_back_r_01}
      boxExtents=<20 30 30>
      constraintSetup {angularTargetAxis=AxisY, angularLimitsMax=<0 20 5>, angularLimitsMin=<0 -120 -5>, angularTarget=<1 0 0>}
      localJointOffset=<4 0 0>
     }]
    relativeSpaceBone {boneName=spine_04}
    simSpaceSettings {maxAngularAcceleration=100, maxAngularVelocity=10, simSpaceAngularAlpha=1}
   }
   n7 = Anim Dynamics - Bone: dyn_vest_front_l_01(ComponentPose=Local To Component(LocalPose=Use cached pose 'startCache'()), Alpha=1, GravityOverride="0,0,-980") {
    angularDampingOverride=1
    angularSpringConstant=20000
    bAngularSpring=true
    bOverrideAngularDamping=true
    bUseGravityOverride=true
    gravityScale=0.35
    simulationSpace=RootRelative
    boundBone {boneName=dyn_vest_front_l_01}
    componentAppliedLinearAccClamp=<4000 4000 4000>
    componentLinearAccScale=<0.7 0.8 0.7>
    componentLinearVelScale=<0.7 0.8 0.7>
    gravityOverride {z=-980}
    physicsBodyDefinitions [{
      boundBone {boneName=dyn_vest_front_l_01}
      boxExtents=<20 30 30>
      constraintSetup {angularTargetAxis=AxisY, angularLimitsMax=<0 0 10>, angularLimitsMin=<0 -90 0>, angularTarget=<1 0 0>}
      localJointOffset=<4 0 0>
     }]
    relativeSpaceBone {boneName=spine_04}
    simSpaceSettings {maxAngularAcceleration=100, maxAngularVelocity=10, simSpaceAngularAlpha=1}
   }
   n8 = Anim Dynamics - Bone: dyn_vest_side_l_01(ComponentPose=@n7, Alpha=1, GravityOverride="0,0,-980") {
    angularDampingOverride=1
    angularSpringConstant=20000
    bAngularSpring=true
    bOverrideAngularDamping=true
    bUseGravityOverride=true
    gravityScale=0.35
    simulationSpace=RootRelative
    boundBone {boneName=dyn_vest_side_l_01}
    componentAppliedLinearAccClamp=<4000 4000 4000>
    componentLinearAccScale=<0.7 0.8 0.7>
    componentLinearVelScale=<0.7 0.8 0.7>
    gravityOverride {z=-980}
    physicsBodyDefinitions [{
      boundBone {boneName=dyn_vest_side_l_01}
      boxExtents=<20 30 30>
      constraintSetup {angularTargetAxis=AxisY, angularLimitsMax=<0 5 10>, angularLimitsMin=<0 -90 -10>, angularTarget=<1 0 0>}
      localJointOffset=<4 0 0>
     }]
    relativeSpaceBone {boneName=spine_04}
    simSpaceSettings {maxAngularAcceleration=100, maxAngularVelocity=10, simSpaceAngularAlpha=1}
   }
   n9 = Anim Dynamics - Bone: dyn_vest_front_r_01(ComponentPose=@n8, Alpha=1, GravityOverride="0,0,-980") {
    angularDampingOverride=1
    angularSpringConstant=20000
    bAngularSpring=true
    bOverrideAngularDamping=true
    bUseGravityOverride=true
    gravityScale=0.35
    simulationSpace=RootRelative
    boundBone {boneName=dyn_vest_front_r_01}
    componentAppliedLinearAccClamp=<4000 4000 4000>
    componentLinearAccScale=<0.7 0.8 0.7>
    componentLinearVelScale=<0.7 0.8 0.7>
    gravityOverride {z=-980}
    physicsBodyDefinitions [{
      boundBone {boneName=dyn_vest_front_r_01}
      boxExtents=<20 30 30>
      constraintSetup {angularTargetAxis=AxisY, angularLimitsMax=<0 2 10>, angularLimitsMin=<0 -90 0>, angularTarget=<1 0 0>}
      localJointOffset=<4 0 0>
     }]
    relativeSpaceBone {boneName=spine_04}
    simSpaceSettings {maxAngularAcceleration=100, maxAngularVelocity=10, simSpaceAngularAlpha=1}
   }
   n10 = Anim Dynamics - Bone: dyn_vest_side_r_01(ComponentPose=@n9, Alpha=1, GravityOverride="0,0,-980") {
    angularDampingOverride=1
    angularSpringConstant=20000
    bAngularSpring=true
    bOverrideAngularDamping=true
    bUseGravityOverride=true
    gravityScale=0.35
    simulationSpace=RootRelative
    boundBone {boneName=dyn_vest_side_r_01}
    componentAppliedLinearAccClamp=<4000 4000 4000>
    componentLinearAccScale=<0.7 0.8 0.7>
    componentLinearVelScale=<0.7 0.8 0.7>
    gravityOverride {z=-980}
    physicsBodyDefinitions [{
      boundBone {boneName=dyn_vest_side_r_01}
      boxExtents=<20 30 30>
      constraintSetup {angularTargetAxis=AxisY, angularLimitsMax=<0 5 10>, angularLimitsMin=<0 -90 -10>, angularTarget=<1 0 0>}
      localJointOffset=<4 0 0>
     }]
    relativeSpaceBone {boneName=spine_04}
    simSpaceSettings {maxAngularAcceleration=100, maxAngularVelocity=10, simSpaceAngularAlpha=1}
   }
   Output Pose(Result=Blend Poses by bool(BlendPose_0=Component To Local(ComponentPose=@n10), BlendPose_1=Component To Local(ComponentPose=@n6)) {blendType=Linear}) {layerGroup=None}
   n11 = Control Rig(Source=Copy Pose From Mesh() {bUseAttachedParent=true}) {controlRigAssetReference {blueprintRigClass=&/M_MED_MossyDeport_CtrlRig.M_MED_MossyDeport_CtrlRig_C}}
   n12 = Anim Dynamics - Bone: dyn_anklet_l_00!(ComponentPose=Local To Component(LocalPose=@n11), Alpha=1, GravityOverride:Vector="0,0,-980") -> (Pose) {
    angularDampingOverride=1
    angularSpringConstant=15000
    bAngularSpring=true
    bOverrideAngularDamping=true
    bUseGravityOverride=true
    gravityScale=0.35
    simulationSpace=RootRelative
    boundBone {boneName=dyn_anklet_l_00}
    componentAppliedLinearAccClamp=<3000 3000 3000>
    componentLinearAccScale=<0.3 0.3 0.3>
    componentLinearVelScale=<0.3 0.3 0.3>
    gravityOverride {z=-980}
    physicsBodyDefinitions [{
      boundBone {boneName=dyn_anklet_l_00}
      boxExtents=<20 30 30>
      constraintSetup {angularLimitsMax=<0 6 6>, angularLimitsMin=<0 -6 -6>, angularTarget=<1 0 0>}
      localJointOffset=<4 0 0>
     }]
    relativeSpaceBone {boneName=foot_l}
    simSpaceSettings {maxAngularAcceleration=50, maxAngularVelocity=5, simSpaceAngularAlpha=0.5}
   }
   n13 = Anim Dynamics - Bone: dyn_anklet_r_01!(ComponentPose=@n12, Alpha=1, GravityOverride:Vector="0,0,-980") -> (Pose) {
    angularDampingOverride=1
    angularSpringConstant=15000
    bAngularSpring=true
    bOverrideAngularDamping=true
    bUseGravityOverride=true
    gravityScale=0.35
    simulationSpace=RootRelative
    boundBone {boneName=dyn_anklet_r_01}
    componentAppliedLinearAccClamp=<3000 3000 3000>
    componentLinearAccScale=<0.3 0.3 0.3>
    componentLinearVelScale=<0.3 0.3 0.3>
    gravityOverride {z=-980}
    physicsBodyDefinitions [{
      boundBone {boneName=dyn_anklet_r_01}
      boxExtents=<20 30 30>
      constraintSetup {angularLimitsMax=<0 6 6>, angularLimitsMin=<0 -6 -6>, angularTarget=<1 0 0>}
      localJointOffset=<4 0 0>
     }]
    relativeSpaceBone {boneName=thigh_r}
    simSpaceSettings {maxAngularAcceleration=50, maxAngularVelocity=5, simSpaceAngularAlpha=0.5}
   }
   n14 = Anim Dynamics - Bone: dyn_pants_r_MSTR!(ComponentPose=@n13, Alpha=1, GravityOverride:Vector="0,0,-980") -> (Pose) {
    angularDampingOverride=1
    angularSpringConstant=20000
    bAngularSpring=true
    bOverrideAngularDamping=true
    bUseGravityOverride=true
    gravityScale=0.35
    simulationSpace=RootRelative
    boundBone {boneName=dyn_pants_r_MSTR}
    componentAppliedLinearAccClamp=<3000 3000 3000>
    componentLinearAccScale=<0.4 0.4 0.4>
    componentLinearVelScale=<0.4 0.4 0.4>
    gravityOverride {z=-980}
    physicsBodyDefinitions [{
      boundBone {boneName=dyn_pants_r_MSTR}
      boxExtents=<30 40 40>
      constraintSetup {angularLimitsMax=<15 15 15>, angularLimitsMin=<-15 -15 -15>, angularTarget=<1 0 0>}
      localJointOffset=<4 0 0>
     }]
    relativeSpaceBone {boneName=pelvis}
    simSpaceSettings {maxAngularAcceleration=100, maxAngularVelocity=10, simSpaceAngularAlpha=1}
   }
   n15 = Anim Dynamics - Bone: dyn_pants_r_outside!(ComponentPose=@n14, Alpha=1, GravityOverride:Vector="0,0,-980") -> (Pose) {
    angularDampingOverride=0.99
    angularSpringConstant=1000
    bAngularSpring=true
    bOverrideAngularDamping=true
    bUseGravityOverride=true
    gravityScale=0.35
    simulationSpace=RootRelative
    boundBone {boneName=dyn_pants_r_outside}
    componentAppliedLinearAccClamp=<3000 3000 3000>
    componentLinearAccScale=<0.4 0.4 0.4>
    componentLinearVelScale=<0.4 0.4 0.4>
    gravityOverride {z=-980}
    physicsBodyDefinitions [{
      boundBone {boneName=dyn_pants_r_outside}
      boxExtents=<10 15 15>
      constraintSetup {angularLimitsMax=<15 15 15>, angularLimitsMin=<-15 -15 -15>, angularTarget=<1 0 0>}
      localJointOffset=<4 0 0>
     }]
    relativeSpaceBone {boneName=pelvis}
    simSpaceSettings {maxAngularAcceleration=100, maxAngularVelocity=10, simSpaceAngularAlpha=1}
   }
   n16 = Anim Dynamics - Bone: dyn_pants_r_inside!(ComponentPose=@n15, Alpha=1, GravityOverride:Vector="0,0,-980") -> (Pose) {
    angularDampingOverride=0.99
    angularSpringConstant=1000
    bAngularSpring=true
    bOverrideAngularDamping=true
    bUseGravityOverride=true
    gravityScale=0.35
    simulationSpace=RootRelative
    boundBone {boneName=dyn_pants_r_inside}
    componentAppliedLinearAccClamp=<3000 3000 3000>
    componentLinearAccScale=<0.4 0.4 0.4>
    componentLinearVelScale=<0.4 0.4 0.4>
    gravityOverride {z=-980}
    physicsBodyDefinitions [{
      boundBone {boneName=dyn_pants_r_inside}
      boxExtents=<10 15 15>
      constraintSetup {angularLimitsMax=<15 15 15>, angularLimitsMin=<-15 -15 -15>, angularTarget=<1 0 0>}
      localJointOffset=<4 0 0>
     }]
    relativeSpaceBone {boneName=pelvis}
    simSpaceSettings {maxAngularAcceleration=100, maxAngularVelocity=10, simSpaceAngularAlpha=1}
   }
   n17 = Anim Dynamics - Bone: dyn_pants_l_MSTR!(ComponentPose=@n16, Alpha=1, GravityOverride:Vector="0,0,-980") -> (Pose) {
    angularDampingOverride=1
    angularSpringConstant=20000
    bAngularSpring=true
    bOverrideAngularDamping=true
    bUseGravityOverride=true
    gravityScale=0.35
    simulationSpace=RootRelative
    boundBone {boneName=dyn_pants_l_MSTR}
    componentAppliedLinearAccClamp=<3000 3000 3000>
    componentLinearAccScale=<0.4 0.4 0.4>
    componentLinearVelScale=<0.4 0.4 0.4>
    gravityOverride {z=-980}
    physicsBodyDefinitions [{
      boundBone {boneName=dyn_pants_l_MSTR}
      boxExtents=<30 40 40>
      constraintSetup {angularLimitsMax=<15 15 15>, angularLimitsMin=<-15 -15 -15>, angularTarget=<1 0 0>}
      localJointOffset=<-4 0 0>
     }]
    relativeSpaceBone {boneName=pelvis}
    simSpaceSettings {maxAngularAcceleration=100, maxAngularVelocity=10, simSpaceAngularAlpha=1}
   }
   n18 = Anim Dynamics - Bone: dyn_pants_l_outside!(ComponentPose=@n17, Alpha=1, GravityOverride:Vector="0,0,-980") -> (Pose) {
    angularDampingOverride=0.99
    angularSpringConstant=1000
    bAngularSpring=true
    bOverrideAngularDamping=true
    bUseGravityOverride=true
    gravityScale=0.35
    simulationSpace=RootRelative
    boundBone {boneName=dyn_pants_l_outside}
    componentAppliedLinearAccClamp=<3000 3000 3000>
    componentLinearAccScale=<0.4 0.4 0.4>
    componentLinearVelScale=<0.4 0.4 0.4>
    gravityOverride {z=-980}
    physicsBodyDefinitions [{
      boundBone {boneName=dyn_pants_l_outside}
      boxExtents=<10 15 15>
      constraintSetup {angularLimitsMax=<15 15 15>, angularLimitsMin=<-15 -15 -15>, angularTarget=<1 0 0>}
      localJointOffset=<4 0 0>
     }]
    relativeSpaceBone {boneName=pelvis}
    simSpaceSettings {maxAngularAcceleration=100, maxAngularVelocity=10, simSpaceAngularAlpha=1}
   }
   n19 = Anim Dynamics - Bone: dyn_pants_l_inside!(ComponentPose=@n18, Alpha=1, GravityOverride:Vector="0,0,-980") -> (Pose) {
    angularDampingOverride=0.99
    angularSpringConstant=1000
    bAngularSpring=true
    bOverrideAngularDamping=true
    bUseGravityOverride=true
    gravityScale=0.35
    simulationSpace=RootRelative
    boundBone {boneName=dyn_pants_l_inside}
    componentAppliedLinearAccClamp=<3000 3000 3000>
    componentLinearAccScale=<0.4 0.4 0.4>
    componentLinearVelScale=<0.4 0.4 0.4>
    gravityOverride {z=-980}
    physicsBodyDefinitions [{
      boundBone {boneName=dyn_pants_l_inside}
      boxExtents=<10 15 15>
      constraintSetup {angularLimitsMax=<15 15 15>, angularLimitsMin=<-15 -15 -15>, angularTarget=<1 0 0>}
      localJointOffset=<4 0 0>
     }]
    relativeSpaceBone {boneName=pelvis}
    simSpaceSettings {maxAngularAcceleration=100, maxAngularVelocity=10, simSpaceAngularAlpha=1}
   }
   n20 = RigidBody!(ComponentPose=@n19, Alpha, OverrideWorldGravity:Vector="0,0,-980") -> (Pose) {
    bOverrideWorldGravity=true
    overridePhysicsAsset=&/M_MED_MossyDeport_LeftLeg_Physics
    simulationSpace=BaseBoneSpace
    baseBoneRef {boneName=pelvis}
    overrideWorldGravity {z=-980}
    simSpaceSettings {
     dampingAlpha=0
     maxAngularAcceleration=50
     maxAngularVelocity=15
     maxLinearAcceleration=4000
     maxLinearVelocity=800
     velocityScaleZ=0.6
     worldAlpha=0.7
     externalLinearDragV=<1.5 1.5 2>
    }
   }
   n21 = RigidBody(ComponentPose=@n20, OverrideWorldGravity="0,0,-980") {
    bOverrideWorldGravity=true
    overridePhysicsAsset=&/M_MED_MossyDeport_RightLeg_Physics
    simulationSpace=BaseBoneSpace
    baseBoneRef {boneName=pelvis}
    overrideWorldGravity {z=-980}
    simSpaceSettings {
     dampingAlpha=0
     maxAngularAcceleration=50
     maxAngularVelocity=15
     maxLinearAcceleration=4000
     maxLinearVelocity=800
     velocityScaleZ=0.6
     worldAlpha=0.7
     externalLinearDragV=<1.5 1.5 2>
    }
   }
   n22 = RigidBody(ComponentPose=@n21, OverrideWorldGravity="0,0,-980") {
    bOverrideWorldGravity=true
    overridePhysicsAsset=&/M_MED_MossyDeport_Tassels_Physics
    simulationSpace=BaseBoneSpace
    baseBoneRef {boneName=pelvis}
    overrideWorldGravity {z=-980}
    simSpaceSettings {
     dampingAlpha=2
     maxAngularAcceleration=50
     maxAngularVelocity=15
     maxLinearAcceleration=4000
     maxLinearVelocity=800
     velocityScaleZ=0.4
     worldAlpha=0.7
     externalLinearDragV=<1.5 1.5 2>
    }
   }
   Save cached pose 'startCache'(Pose=Component To Local(ComponentPose=@n22)) {cacheName=startCache}
  ]}
 EventGraph {nodes [
   Event BlueprintUpdateAnimation()
   > n1 = Sequence()
   > n2 = "Gravity Override State Function (Parameter Asset) (Nativized)"(
    CharacterPartAnimInstance=Self-Reference()
    GameTimeInSeconds=0
    Parameters=&/Fortnite_MossyDeport_Parameters
   )
   > Set Gravity Override(Gravity Override=@n2)
   Set frontendAlpha(execute=@n1.then_1, frontendAlpha:double=SelectFloat(A=0.35, B=1, bPickA=Get bIsInFrontEnd())) -> (Output_Get:double)
   Event OnCharacterCustomizationLoaded()
   n3 = CheckForBackbling()
   > Set isWearingPack(isWearingPack:bool=OR Boolean!(A=@n3, B=@n3.bIsBackblingACape, C:bool=@n3.bIsBackblingLong) -> (ReturnValue)) -> (Output_Get:bool)
   > Set isWearingCape(isWearingCape=@n3.bIsBackblingLong)
  ]}
]
variables [
 frontendAlpha {type=number}
 Gravity Override {type=Vector}
 isWearingCape {type=boolean}
 isWearingPack {type=boolean}
]
