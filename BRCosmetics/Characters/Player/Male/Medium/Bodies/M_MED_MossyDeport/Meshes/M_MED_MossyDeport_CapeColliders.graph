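# Physics asset of cape colliders for the M_MED_MossyDeport male player body, by its name a collision proxy for cape simulation.
# Capsules cover the pelvis, spine, neck and clavicles, with a sphere on the head and extra overlapping capsules on spine_02 and spine_05.
# Small capsules also follow the dyn_backTassel bone chains.
# Arm and leg bodies have no shapes, and no constraints link any bodies.
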PhysicsAsset /BRCosmetics/Characters/Player/Male/Medium/Bodies/M_MED_MossyDeport/Meshes/M_MED_MossyDeport_CapeColliders
bodies {
 calf_l []
 calf_r []
 clavicle_l [{shapeType=Capsule, length=9.0192, radius=9.8871, center=<9.4266 -1.1806 -2.4977>, rotation=<rot 69.3297 1.4476 176.9462>}]
 clavicle_r [{shapeType=Capsule, length=9.0192, radius=9.8871, center=<-9.4266 1.1806 2.4977>, rotation=<rot 69.3297 1.4476 -3.0538>}]
 dyn_backTassel_c_01_l [{
   shapeType=Capsule
   length=3.4088
   radius=0.8018
   shapeName=dyn_backTassel_c_01_l_capsule
   center=<2.9731 -0.0486 -0.0208>
   rotation=<rot 0 -91.9517 -89.6567>
  }]
 dyn_backTassel_c_01_r [{
   shapeType=Capsule
   length=3.0359
   radius=0.7295
   shapeName=dyn_backTassel_c_01_r_capsule
   center=<3.12 -0.0469 -0.1239>
   rotation=<rot 0 90.5534 -88.3116>
  }]
 dyn_backTassel_c_02_l [{
   shapeType=Capsule
   length=4.2428
   radius=1.0305
   shapeName=dyn_backTassel_c_02_l_capsule
   center=<2.6859 0.1662 -0.0379>
   rotation=<rot 0 89.3277 -89.1662>
  }]
 dyn_backTassel_c_02_r [{
   shapeType=Capsule
   length=4.3321
   radius=1.0899
   shapeName=dyn_backTassel_c_02_r_capsule
   center=<2.6652 -0.0084 0.1383>
   rotation=<rot 0 -90.3092 -89.7803>
  }]
 dyn_backTassel_c_03_l [{
   shapeType=Capsule
   length=2.8919
   radius=1.6719
   shapeName=dyn_backTassel_c_03_l_capsule
   center=<1.9532 0.0927 0.0668>
   rotation=<rot 0 -90.269 -86.0757>
  }]
 dyn_backTassel_c_03_r [{
   shapeType=Capsule
   length=2.9731
   radius=1.7526
   shapeName=dyn_backTassel_c_03_r_capsule
   center=<2.0925 0.0748 0.1333>
   rotation=<rot 0 -90.6375 -89.5713>
  }]
 dyn_backTassel_c_04_l [{
   shapeType=Capsule
   length=2.597
   radius=2.1901
   shapeName=dyn_backTassel_c_04_l_capsule
   center=<1.6304 0.083 -0.1928>
   rotation=<rot 0 91.2527 -85.9034>
  }]
 dyn_backTassel_c_04_r [{
   shapeType=Capsule
   length=3.4089
   radius=2.3344
   shapeName=dyn_backTassel_c_04_r_capsule
   center=<1.107 0.0401 0.1617>
   rotation=<rot 0 89.1654 -87.0696>
  }]
 dyn_backTassel_l_01 [{
   shapeType=Capsule
   length=3.4046
   radius=2.2938
   shapeName=dyn_backTassel_l_01_capsule
   center=<3.1453 0.3981 0.2415>
   rotation=<rot 0 -90.914 -85.9618>
  }]
 dyn_backTassel_l_02 [{
   shapeType=Capsule
   length=3.2684
   radius=2.9993
   shapeName=dyn_backTassel_l_02_capsule
   center=<2.2962 0.3851 0.0512>
   rotation=<rot 0 92.3018 -88.3282>
  }]
 dyn_backTassel_r_01 [{
   shapeType=Capsule
   length=3.381
   radius=2.2932
   shapeName=dyn_backTassel_r_01_capsule
   center=<3.1721 0.3995 -0.2457>
   rotation=<rot 0 89.0503 -85.8413>
  }]
 dyn_backTassel_r_02 [{
   shapeType=Capsule
   length=3.2684
   radius=2.9998
   shapeName=dyn_backTassel_r_02_capsule
   center=<2.2963 0.3845 -0.0496>
   rotation=<rot 0 -87.7091 -88.3587>
  }]
 foot_l []
 foot_r []
 head [{shapeType=Sphere, radius=17.3125, center=<11.028 -0.174 0>}]
 lowerarm_l []
 lowerarm_r []
 neck_02 [{shapeType=Capsule, length=13.4375, radius=7.3175, center=<-0.2642 -0.3408 0>, rotation=<rot 90 -171.1193 -175.3978>}]
 pelvis [{shapeType=Capsule, length=5.6456, radius=17.5388, center=<-11.4887 -6.8119 0>}]
 spine_01 [{shapeType=Capsule, length=7.9438, radius=13.5586, center=<-1.7965 0.1884 0>}]
 spine_02 [
  {shapeType=Capsule, length=7.3264, radius=13.6322, center=<-1.711 -0.1851 0.0001>}
  {shapeType=Capsule, length=29.3956, radius=11.5807, center=<2.5024 0.4992 3.9825>, rotation=<rot 90 0.0049 0.0049>}
  {shapeType=Capsule, length=29.875, radius=11.1013, center=<2.4704 0.4991 -3.9825>, rotation=<rot 90 -0.0396 -0.0396>}
 ]
 spine_03 [{shapeType=Capsule, length=12.2107, radius=13.7031, center=<1.7685 0.2396 0.0001>}]
 spine_04 [{shapeType=Capsule, length=9.4404, radius=14.1325, center=<2.766 0.7765 0.0002>}]
 spine_05 [
  {shapeType=Capsule, length=8.0226, radius=15.695, center=<-2.2795 1.8109 0>}
  {shapeType=Capsule, length=13.9518, radius=13.9425, center=<0.4695 -1.7418 0>}
 ]
 thigh_l []
 thigh_r []
 upperarm_l []
 upperarm_r []
}
constraints []
